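; Part modifier for the BirthMitten quadruped companion cosmetic that lights its collar and helmet in response to voice-over audio.
; It caches the owner's companion anim instance on BeginPlay, then each tick sets a frontend flag on part materials and runs VoActivityCheck.
; That check thresholds the VO audio envelope, higher for collar than helmet, to drive material scalars and idle VFX parameters.
; It also plays a DoOnce-gated light-on sound at the owner's location.
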(event EventBeginPlay
  (|CallParentFunction)
  (bind _ascompanion_quadruped_base_anim_instance (Utilities|Casting|CastToCompanionQuadrupedBaseAnimInstance (Components|SkeletalMesh|GetAnimInstance (Actor|GetComponentbyClass (Actor|GetOwner) "/Script/Engine.SkeletalMeshComponent"))))
  (Variables|CustomCharacterPartModifier|SetCompanion%20Anim%20Instance _ascompanion_quadruped_base_anim_instance)
  (CallFunction|VoActivityCheck))

(event EventTick
  (params (DeltaSeconds :type Float%20%28single-precision%29))
  (|CallParentFunction DeltaSeconds)
  (Rendering|Material|SetScalarParameterValueonMaterials (CustomCharacter|GetSkeletalMeshforPartonOwner) "IsFrontend" (Math|Conversions|ToFloat(Boolean) (Reactive|Frontend|IsCosmeticObjectDisplayedinFrontendContext 0)))
  (CallFunction|VoActivityCheck))

(event Custom|VoActivityCheck
  (bind _returnvalue (CallFunction|GetIdleVFXComponent))
  (bind _returnvalue_2 (CustomCharacter|GetSkeletalMeshforPartonOwner))
  (bind _companion_anim_instance (Variables|CustomCharacterPartModifier|GetCompanion%20Anim%20Instance))
  (bind _bisvoactive (Variables|/Script/CosmeticCompanionsRuntime.FortPlayerCompanionAnimInstance|GetbIsVOActive _companion_anim_instance))
  (bind _audioenvelopevalue (Variables|/Script/CosmeticCompanionsRuntime.FortPlayerCompanionAnimInstance|GetAudioEnvelopeValue _companion_anim_instance))
  (bind _returnvalue_3 (>= _audioenvelopevalue 0.75))
  (bind _returnvalue_4 (and _returnvalue_3 _bisvoactive))
  (bind _returnvalue_5 (Math|Conversions|ToFloat(Boolean) _returnvalue_4))
  (bind _returnvalue_6 (>= _audioenvelopevalue 0.2))
  (bind _returnvalue_7 (and _bisvoactive _returnvalue_6))
  (Effects|Components|ParticleSystem|SetBoolParameter _returnvalue "bIsVOActive" _returnvalue_7)
  (Effects|Components|ParticleSystem|SetFloatParameter _returnvalue "VOCurve" _returnvalue_5)
  (Rendering|Material|SetScalarParameterValueonMaterials _returnvalue_2 "VOActivity - Collar" _returnvalue_5)
  (Rendering|Material|SetScalarParameterValueonMaterials _returnvalue_2 "VOActivity - Helmet" (Math|Conversions|ToFloat(Boolean) _returnvalue_7))
  (if _returnvalue_7
    _)
  (Utilities|FlowControl|DoOnce false
    (:Completed
      (Audio|PlaySoundatLocation "/CosmeticCompanions/Assets/Quadruped/BirthMitten/Audio/Emote/MSS_Mimosa_BirthMitten_LightOn.MSS_Mimosa_BirthMitten_LightOn" (Transformation|GetActorLocation (Actor|GetOwner))))))

(event Custom|TriggerBlip)

(fn ConstructionScript
  (|CallParentFunction))
