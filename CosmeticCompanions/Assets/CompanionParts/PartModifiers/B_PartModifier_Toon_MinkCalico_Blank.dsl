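; Cosmetic part modifier for the Toon MinkCalico companion part, adding helmet audio mutation and a toon frontend LUT replacement.
; On BeginPlay it registers its helmet audio mutator with the Audio Mutators subsystem, then appends an actor-instance selector targeting its owner.
; EndPlay removes the mutator.
; After the parent's post-res-in handling, it calls ToonFrontendLUTReplace.
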(event EventBeginPlay
  (bind _helmet_audio_mutator (Variables|Default|GetHelmet%20Audio%20Mutator))
  (|CallParentFunction)
  (Audio|Mutator|AddMutator (AudioEngineSubsystems|GetAudioMutatorsSubsystem) _helmet_audio_mutator)
  (bind _returnvalue (Utilities|InstancedStruct|MakeInstancedStruct (Utilities|Struct|MakeActorInstance (Actor|GetOwner))))
  (Utilities|Array|Add (Variables|/Script/AudioMutators.AudioMutator|GetSelectors _helmet_audio_mutator) _returnvalue))

(event EventEndPlay
  (params (EndPlayReason :type EEndPlayReason%20Enum :default "Destroyed"))
  (Audio|Mutator|RemoveMutator (AudioEngineSubsystems|GetAudioMutatorsSubsystem) (Variables|Default|GetHelmet%20Audio%20Mutator)))

(event EventOnPostResIn
  (|CallParentFunction)
  (Utilities|ToonFrontendLUTReplace))

(fn ConstructionScript
  (|CallParentFunction))
